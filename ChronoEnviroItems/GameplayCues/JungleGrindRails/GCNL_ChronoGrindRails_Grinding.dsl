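; Looping gameplay cue that plays VFX and audio while a player pawn grinds on Chrono jungle grind rails.
; On start it binds grinding provider delegates that set particle parameters for sprint, booster mode, jump and rail color.
; It plays start, loop, jump-off and lean sounds, feeding native audio logic cached references and a feet-landed flag.
; Non-teammates get a looping sound indicator; removal unbinds delegates and fades out start and loop audio.
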(event Custom|OnFeetLanded
  (Audio|AllowAudio))

(event EventOnLoopingStart(EitherVFX-preferthisversion)
  (params (MyTarget :type Actor%20Object%20Reference) (Parameters :type Gameplay%20Cue%20Parameters%20Structure%20%28by%20ref%29) (ParticleComponents :type Array%20of%20FXSystem%20Component%20Object%20References) (AudioComponents :type Array%20of%20Audio%20Component%20Object%20References))
  (bind _returnvalue (Actor|GetComponentbyClass MyTarget "/GrindRail/Blueprints/BP_FortPawnGrindingProvider.BP_FortPawnGrindingProvider_C"))
  (bind _grindrailfx (Variables|Default|GetGrindRailFX))
  (bind _playerpawn_0 (Variables|Default|GetPlayerPawn_0))
  (bind _cachedgrindingprovider (Variables|Default|GetCachedGrindingProvider))
  (bind _cachedgrindingprovider_2 (Variables|Default|GetCachedGrindingProvider))
  (Utilities|FlowControl|Sequence
    (:then_0
      (Variables|Default|SetGrindRail (Class|FortPawnComponentGrindingProvider|GetCurrentGrindingRailActor _returnvalue))
      (bind _output_get (Variables|Default|SetCachedGrindingProvider _returnvalue))
      (EventDispatchers|BindEventtoFeetLandedDelegate _output_get (delegate Custom|OnFeetLanded))
      (bind _asfort_player_pawn (Utilities|Casting|CastToFortPlayerPawn MyTarget))
      (Variables|Default|SetPlayerPawn_0 _asfort_player_pawn)
      (Audio|GrindAudioState true)
      (bind _returnvalue_2 (Development|GetConsoleVariableFloatValue "Grinding.NotRenderedDisableEffectsTime"))
      (Variables|Default|SetNotRenderedDisableEffectsTime _returnvalue_2)
      (EventDispatchers|BindEventtoEndedGrindingDelegate (Variables|Default|GetCachedGrindingProvider) (EventDispatchers|CreateEvent :!event OnJumpedOff)))
    (:then_1
      (bind _output_get_2 (Variables|Default|SetGrindRailFX (Utilities|Array|Get(acopy) ParticleComponents)))
      (bind _output_get_3 (Variables|Audio|SetGrindAudioLoop (Utilities|Array|Get(acopy) AudioComponents)))
      (Audio|Components|Audio|FadeIn _output_get_3 0.5 1.0 0.0 "Logarithmic" :!class /Script/Engine.AudioComponent)
      (CallFunction|CacheReferences _output_get_3 _output_get_2 _playerpawn_0 (Variables|Default|GetCachedGrindingProvider))
      (Audio|Parameter|SetBooleanParameter _output_get_3 "IsLocal" (Pawn|IsLocallyViewed _playerpawn_0) :!class /Script/AudioExtensions.AudioParameterControllerInterface))
    (:then_2
      ;[1 1 1] Sprint
      (EventDispatchers|BindEventtoSprintingStateChangedDelegate _cachedgrindingprovider (delegate Custom|OnGrindRailSprint))
      (Effects|Components|ParticleSystem|SetBoolParameter (Variables|Default|GetGrindRailFX) "Sprinting" (Class|FortPawnComponentGrindingProvider|IsSprinting _cachedgrindingprovider)))
    (:then_3
      ;[1 1 1] Booster Mode
      (EventDispatchers|BindEventtoBoosterModeChangedDelegate _cachedgrindingprovider_2 (delegate Custom|OnGrindRailBoosterMode))
      (switch Utilities|FlowControl|Switch|SwitchonEGrindRailBoosterMode (Class|FortPawnComponentGrindingProvider|GetCurrentBoosterMode _cachedgrindingprovider_2)
        (:SpeedUp
          (Effects|Components|ParticleSystem|SetBoolParameter _grindrailfx "Boosting" true))
        (:SlowDown
          (Effects|Components|ParticleSystem|SetBoolParameter _grindrailfx "SlowDown" true))))
    (:then_4
      ;[0.9017 0.905 1] Jump
      (EventDispatchers|BindEventtoGrindRailChangedDelegate (Variables|Default|GetCachedGrindingProvider) (delegate Custom|HandleGrindRailChanged))
      (Effects|Components|ParticleSystem|SetBoolParameter (Variables|Default|GetGrindRailFX) "Jumped" false)
      (bind _asbga_grind_rail (Utilities|Casting|CastToBGA_GrindRail (Variables|Default|GetGrindRail)))
      (Effects|Components|ParticleSystem|SetIntParameter (Variables|Default|GetGrindRailFX) "ColorIndex" (Math|Conversions|ToInteger(Byte) (Variables|/GrindRail/Blueprints/BGA_GrindRail.BGA_GrindRail_C|GetColor_Select_Enum _asbga_grind_rail))))))

(event Custom|OnGrindRailSprint
  (params (bNewIsSprinting :type Boolean))
  ;[1 1 1] Sprint
  (Effects|Components|ParticleSystem|SetBoolParameter (Variables|Default|GetGrindRailFX) "Sprinting" bNewIsSprinting))

(event EventOnRemoval(EitherVFX-preferthisversion)
  (params (MyTarget :type Actor%20Object%20Reference) (Parameters :type Gameplay%20Cue%20Parameters%20Structure%20%28by%20ref%29) (ParticleComponents :type Array%20of%20FXSystem%20Component%20Object%20References) (AudioComponents :type Array%20of%20Audio%20Component%20Object%20References) (BurstCameraShakeInstance :type Camera%20Shake%20Base%20Object%20Reference) (BurstDecalInstance :type Decal%20Actor%20Object%20Reference))
  ;[1 1 1] Always MyTarget as PlayerPawn to remove effects stably
  (bind _asfort_player_pawn (Utilities|Casting|CastToFortPlayerPawn MyTarget))
  (Variables|Default|SetPlayerPawn_0 _asfort_player_pawn)
  (bind _output_get (Variables|Default|SetCachedGrindingProvider (Actor|GetComponentbyClass "/GrindRail/Blueprints/BP_FortPawnGrindingProvider.BP_FortPawnGrindingProvider_C")))
  (EventDispatchers|UnbindEventfromEndedGrindingDelegate _output_get (EventDispatchers|CreateEvent :!event OnJumpedOff))
  (EventDispatchers|UnbindEventfromFeetLandedDelegate _output_get (delegate Custom|OnFeetLanded))
  (EventDispatchers|UnbindEventfromSprintingStateChangedDelegate _output_get (delegate Custom|OnGrindRailSprint))
  (EventDispatchers|UnbindEventfromBoosterModeChangedDelegate _output_get (delegate Custom|OnGrindRailBoosterMode))
  (EventDispatchers|UnbindEventfromGrindRailChangedDelegate _output_get (delegate Custom|HandleGrindRailChanged))
  (Audio|GrindAudioState false))

(event Custom|OnGrindRailBoosterMode
  (params (PreviousBoosterMode :type EGrindRailBoosterMode%20Enum :default "SpeedUp"))
  (bind _grindrailfx (Variables|Default|GetGrindRailFX))
  (bind _grindrailfx_2 (Variables|Default|GetGrindRailFX))
  (bind _cachedgrindingprovider (Variables|Default|GetCachedGrindingProvider))
  ;[1 1 1] Booster Mode
  (Utilities|FlowControl|Sequence
    (:then_0
      (Utilities|FlowControl|DelayUntilNextTick _cachedgrindingprovider)
      (switch Utilities|FlowControl|Switch|SwitchonEGrindRailBoosterMode (Class|FortPawnComponentGrindingProvider|GetCurrentBoosterMode _cachedgrindingprovider)
        (:SpeedUp
          (Effects|Components|ParticleSystem|SetBoolParameter _grindrailfx_2 "Boosting" true))
        (:SlowDown
          (Effects|Components|ParticleSystem|SetBoolParameter _grindrailfx_2 "SlowDown" true))))
    (:then_1
      (switch Utilities|FlowControl|Switch|SwitchonEGrindRailBoosterMode PreviousBoosterMode
        (:SpeedUp
          (Effects|Components|ParticleSystem|SetBoolParameter _grindrailfx "Boosting" false))
        (:SlowDown
          (Effects|Components|ParticleSystem|SetBoolParameter _grindrailfx "SlowDown" false))))))

(event Custom|HandleGrindRailChanged
  (params (OldRail :type Fort%20Grind%20Rail%20Object%20Reference) (NewRail :type Fort%20Grind%20Rail%20Object%20Reference))
  (bind _grindrailfx (Variables|Default|GetGrindRailFX))
  ;[0.9017 0.905 1] Jump
  ;[1 1 1] to handle rail switching and remote replication
  (bind _output_get (Variables|Default|SetGrindRail NewRail))
  (Utilities|IsValid _output_get
    (:Is%20Valid
      (Effects|Components|ParticleSystem|SetBoolParameter _grindrailfx "Jumped" false)
      (bind _asbga_grind_rail (Utilities|Casting|CastToBGA_GrindRail (Variables|Default|GetGrindRail)))
      (Effects|Components|ParticleSystem|SetIntParameter (Variables|Default|GetGrindRailFX) "ColorIndex" (Math|Conversions|ToInteger(Byte) (Variables|/GrindRail/Blueprints/BGA_GrindRail.BGA_GrindRail_C|GetColor_Select_Enum _asbga_grind_rail))))
    (:Is%20Not%20Valid
      (Effects|Components|ParticleSystem|SetBoolParameter _grindrailfx "Jumped" true))))

(event Audio|Native|EventOnForwardChanged
  (params (bNewState :type Boolean))
  (if bNewState
    (bind _leanforwardaudio (Variables|Default|GetLeanForwardAudio)
      (:then
        (Audio|Components|Audio|FadeOut _leanforwardaudio 0.15 0.0 "Linear" :!class /Script/Engine.AudioComponent))
      (:else
        _))
    (bind _returnvalue (Audio|SpawnSoundAttached :Sound (Variables|Audio|GetLeanForwardSound) :AttachToComponent (Variables|/Script/Engine.Actor|GetRootComponent (Variables|Default|GetPlayerPawn_0)) :AttachPointName "None" :LocationType "KeepRelativeOffset" :bStopWhenAttachedToDestroyed true :VolumeMultiplier 1.0 :PitchMultiplier 1.0 :StartTime 0.0 :bAutoDestroy true))
    (Variables|Default|SetLeanForwardAudio _returnvalue)))

(fn AllowAudio
  (Variables|Audio|Native|SetbFeetHasLanded true) ; Necessary for C++ audio to work
)

(fn GrindAudioState
  (params (Start :type Boolean))
  (Variables|Audio|SetbGrinding (Class|FortPawnComponentGrindingProvider|IsGrinding (Variables|Default|GetCachedGrindingProvider)))
  (Variables|Default|GetPlayerPawn_0
    (:then
      (if Start
        (bind _grindaudiostart (Variables|Audio|GetGrindAudioStart)
          (:then
            (Audio|Components|Audio|FadeOut _grindaudiostart 0.15 0.0 "Linear" :!class /Script/Engine.AudioComponent))
          (:else
            _))
        (bind _returnvalue (Audio|SpawnSoundAttached :Sound (Variables|Audio|GetStartLoopSound) :AttachToComponent (Variables|/Script/Engine.Actor|GetRootComponent (Variables|Default|GetPlayerPawn_0)) :AttachPointName "None" :LocationType "KeepRelativeOffset" :bStopWhenAttachedToDestroyed true :VolumeMultiplier 1.0 :PitchMultiplier 1.0 :StartTime 0.0 :bAutoDestroy true))
        (Variables|Audio|SetGrindAudioStart _returnvalue)
        (Audio|AllowAudio)
        (if (not (Team|OnLocalPlayersTeam (Variables|Default|GetPlayerPawn_0) (Variables|Default|GetPlayerPawn_0)))
          (SoundIndicator|StartTargetedSoundIndicatorLooping (Variables|Default|GetFortSoundIndicator) (Variables|Default|GetPlayerPawn_0) "/ChronoEnviroItems/JungleGrindRails/Audio/GrindRail_Jungle_Loop_Cue.GrindRail_Jungle_Loop_Cue" 1.0 0))
        (else
          (Variables|Audio|Native|SetbFeetHasLanded false)
          (SoundIndicator|StopSoundIndicatorLooping (Variables|Default|GetFortSoundIndicator))
          (bind _grindaudiostart_2 (Variables|Audio|GetGrindAudioStart)
            (:then
              (Audio|Components|Audio|FadeOut _grindaudiostart_2 0.25 0.0 "Linear" :!class /Script/Engine.AudioComponent))
            (:else
              _))
          (bind _grindaudioloop (Variables|Audio|GetGrindAudioLoop)
            (:then
              (Audio|Components|Audio|FadeOut _grindaudioloop 0.5 0.0 "Linear" :!class /Script/Engine.AudioComponent))
            (:else)))))
    (:else)))

(fn OnJumpedOff
  (if (Variables|Audio|GetbGrinding)
    (bind _jumpoffaudio (Variables|Audio|GetJumpOffAudio)
      (:then
        (Audio|Components|Audio|FadeOut _jumpoffaudio 0.15 0.0 "Linear" :!class /Script/Engine.AudioComponent))
      (:else
        _))
    (bind _returnvalue (Audio|SpawnSoundAttached (Variables|Audio|GetJumpUpSound) (Variables|/Script/Engine.Actor|GetRootComponent (Variables|Default|GetPlayerPawn_0))))
    (Variables|Audio|SetJumpOffAudio _returnvalue)))

(fn ConstructionScript)
