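; Looping gameplay cue notify for sprinting on pumpkin grind vines, playing boost audio on the player pawn.
; When the loop starts, it spawns the Boost sound attached to the pawn's mesh; on removal, it spawns BoostEnd the same way.
; Both components are stored, and each event fades out any existing boost and boost-end audio over a quarter second, so rapid toggling crossfades.
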(event EventOnLoopingStart(EitherVFX-preferthisversion)
  (params (MyTarget :type Actor%20Object%20Reference) (Parameters :type Gameplay%20Cue%20Parameters%20Structure%20%28by%20ref%29) (ParticleComponents :type Array%20of%20FXSystem%20Component%20Object%20References) (AudioComponents :type Array%20of%20Audio%20Component%20Object%20References))
  (bind _asfort_player_pawn (Utilities|Casting|CastToFortPlayerPawn MyTarget))
  (bind _boostaudio (Variables|Default|GetBoostAudio)
    (:then
      (Audio|Components|Audio|FadeOut _boostaudio 0.25 0.0 "Linear" :!class /Script/Engine.AudioComponent))
    (:else
      _))
  (bind _returnvalue (Audio|SpawnSoundAttached (Variables|Default|GetBoost) (Variables|/Script/Engine.Character|GetMesh _asfort_player_pawn)))
  (Variables|Default|SetBoostAudio _returnvalue)
  (bind _boostendaudio (Variables|Default|GetBoostEndAudio)
    (:then
      (Audio|Components|Audio|FadeOut _boostendaudio 0.25 0.0 "Linear" :!class /Script/Engine.AudioComponent))
    (:else)))

(event EventOnRemoval(EitherVFX-preferthisversion)
  (params (MyTarget :type Actor%20Object%20Reference) (Parameters :type Gameplay%20Cue%20Parameters%20Structure%20%28by%20ref%29) (ParticleComponents :type Array%20of%20FXSystem%20Component%20Object%20References) (AudioComponents :type Array%20of%20Audio%20Component%20Object%20References) (BurstCameraShakeInstance :type Camera%20Shake%20Base%20Object%20Reference) (BurstDecalInstance :type Decal%20Actor%20Object%20Reference))
  (bind _asfort_player_pawn (Utilities|Casting|CastToFortPlayerPawn MyTarget))
  (bind _boostendaudio (Variables|Default|GetBoostEndAudio)
    (:then
      (Audio|Components|Audio|FadeOut _boostendaudio 0.25 0.0 "Linear" :!class /Script/Engine.AudioComponent))
    (:else
      _))
  (bind _returnvalue (Audio|SpawnSoundAttached (Variables|Default|GetBoostEnd) (Variables|/Script/Engine.Character|GetMesh _asfort_player_pawn)))
  (Variables|Default|SetBoostEndAudio _returnvalue)
  (bind _boostaudio (Variables|Default|GetBoostAudio)
    (:then
      (Audio|Components|Audio|FadeOut _boostaudio 0.25 0.0 "Linear" :!class /Script/Engine.AudioComponent))
    (:else)))

(fn ConstructionScript)
